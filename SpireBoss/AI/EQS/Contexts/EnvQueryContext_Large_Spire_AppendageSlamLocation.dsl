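; EQS context for the Large Spire boss that supplies a single location, the impact point of its appendage slam, to environment queries.
; ProvideSingleLocation casts the querier object to the GA_Large_Spire_AppendageSlam gameplay ability and returns the location of that ability's ImpactTransform.
; If the querier is not that ability, it returns the world origin.
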(fn ProvideSingleLocation
  (params (QuerierObject :type Object%20Reference) (QuerierActor :type Actor%20Object%20Reference))
  (bind _asga_large_spire_appendage_slam (Utilities|Casting|CastToGA_Large_Spire_AppendageSlam QuerierObject)
    (:then
      (return (.location (Variables|/SpireBoss/Abilities/GA_Large_Spire_AppendageSlam.GA_Large_Spire_AppendageSlam_C|GetImpactTransform _asga_large_spire_appendage_slam))))
    (:CastFailed
      (return "0, 0, 0"))))
