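; EQS context for the Large Spire boss that supplies the world locations of its armature appendages to queries.
; It casts the querier to the spire's projectile context actor to reach the owning spire.
; The location set gathers unique world-space bone positions from the skeletal mesh, using each appendage struct's bone name.
; Despite the asset's name, the actor set returns the spire's sensed player pawns.
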(fn ProvideActorsSet
  (params (QuerierObject :type Object%20Reference) (QuerierActor :type Actor%20Object%20Reference))
  (bind _asbga_large_spire_proj_context (Utilities|Casting|CastToBGA_Large_Spire_Proj_Context QuerierActor)
    (:then
      (return (Variables|/SpireBoss/GameplayActors/BGA_Large_Spire.BGA_Large_Spire_C|GetSensedPlayerPawns (Variables|/SpireBoss/GameplayActors/BGA_Large_Spire_Proj_Context.BGA_Large_Spire_Proj_Context_C|GetSpireActor _asbga_large_spire_proj_context))))
    (:CastFailed
      (return 0))))

(fn ProvideLocationsSet
  (params (QuerierObject :type Object%20Reference) (QuerierActor :type Actor%20Object%20Reference))
  (bind _asbga_large_spire_proj_context (Utilities|Casting|CastToBGA_Large_Spire_Proj_Context QuerierActor)
    (:then
      (Utilities|Array|Clear (LocalVariables|GetAppendageLocationSet))
      (LocalVariables|SetSpireSkeletalMesh (Variables|/SpireBoss/GameplayActors/BGA_Large_Spire.BGA_Large_Spire_C|GetSkeletalMeshComponentBudgeted (Variables|/SpireBoss/GameplayActors/BGA_Large_Spire_Proj_Context.BGA_Large_Spire_Proj_Context_C|GetSpireActor _asbga_large_spire_proj_context)))
      (for _array_element (Variables|/SpireBoss/GameplayActors/BGA_Large_Spire.BGA_Large_Spire_C|GetSpireAppendageStruct (Variables|/SpireBoss/GameplayActors/BGA_Large_Spire_Proj_Context.BGA_Large_Spire_Proj_Context_C|GetSpireActor _asbga_large_spire_proj_context))
        (bind _returnvalue_location (Math|Transform|BreakTransform (Transformation|GetBoneTransform (LocalVariables|GetSpireSkeletalMesh) (.BoneName_5_38CD39514D738E05454978B3E574F8A0 (Utilities|Struct|BreakSpireAppendageStruct _array_element)) "RTS_World")))
        (Utilities|Array|AddUnique (LocalVariables|GetAppendageLocationSet) _returnvalue_location))
      (return (LocalVariables|GetAppendageLocationSet)))
    (:CastFailed
      (return 0))))
