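; EQS context for the Large Spire boss that supplies the player pawns the spire has sensed as a set of actors.
; It overrides ProvideActorsSet and expects the querier to be a BGA_Large_Spire_Proj_Context actor, not the spire itself.
; From that actor it gets the owning spire and returns its GetSensedPlayerPawns result.
; If the cast fails, it provides no targets.
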(fn ProvideActorsSet
  (params (QuerierObject :type Object%20Reference) (QuerierActor :type Actor%20Object%20Reference))
  (bind _asbga_large_spire_proj_context (Utilities|Casting|CastToBGA_Large_Spire_Proj_Context QuerierActor)
    (:then
      (return (Variables|/SpireBoss/GameplayActors/BGA_Large_Spire.BGA_Large_Spire_C|GetSensedPlayerPawns (Variables|/SpireBoss/GameplayActors/BGA_Large_Spire_Proj_Context.BGA_Large_Spire_Proj_Context_C|GetSpireActor _asbga_large_spire_proj_context))))
    (:CastFailed
      (return 0))))
